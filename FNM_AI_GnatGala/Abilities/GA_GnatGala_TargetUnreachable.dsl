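; Gameplay ability that makes the GnatGala AI back away from targets it cannot reach.
; While Threatened, a timer polls blackboard reachability for grounded or surface-swimming targets.
; After sustained unreachability, it moves the leash away from the target, pauses target focusing and pushes a behavior control.
; A return timer, Unaware alert or target change restores the saved leash and focus, with ResetLeashHack clearing a blackboard execution status on each leash change.
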(event Ability|EventActivateAbility
  (bind _aicontroller (Variables|InternalData|GetAIController))
  (Utilities|FlowControl|Delay 0.2)
  (if (ScalableFloat|GetValueasBool (Variables|Default|GetUseRunFromUnreachable))
    (bind _outfortplayerpawn (CallFunction|GetAvatarasValidFortPlayerPawn)
      (:ValidFortPlayerPawn
        (bind _output_get (Variables|InternalData|SetAIPawn _outfortplayerpawn))
        (bind _asfort_athena_aibot_controller (Utilities|Casting|CastToFortAthenaAIBotController (Pawn|GetController _output_get))
          (:then
            (bind _output_get_2 (Variables|InternalData|SetAIController _asfort_athena_aibot_controller))
            (AlertLevel|BindEventtoOnAlertLevelChangedEvent _output_get_2 (EventDispatchers|CreateEvent :!event Custom|AIAlertLevelChanged))
            (Perception|BindEventtoOnCurrentTargetChangedEvent _aicontroller (EventDispatchers|CreateEvent :!event Custom|AICurrentTargetChanged))
            (Variables|InternalData|SetAIBlackBoard (AI|GetBlackboard _aicontroller))
            (bind (_leashcenter _leashinnerradius _leashoutterradius) (CallFunction|GetLeashData))
            (Variables|InternalData|SetOGLeashCenter _leashcenter)
            (Variables|InternalData|SetOGLeashInnerRadius _leashinnerradius)
            (Variables|InternalData|SetOGLeashOutterRadius _leashoutterradius)
            (Variables|InternalData|SetIsMovingAway false)
            (Variables|InternalData|SetTimeWhenStartedMovingAway (Utilities|Time|GetGameTimeinSeconds))
            (Variables|InternalData|SetStartedUnreachableCheck false))
          (:CastFailed
            _)))
      (:AvatarCastFailed
        _))
    (Ability|EndAbility)
    (else
      _)))

(event Ability|EventOnEndAbility
  (params (bWasCancelled :type Boolean))
  (bind _aicontroller (Variables|InternalData|GetAIController))
  (Utilities|Time|ClearandInvalidateTimerbyHandle (Variables|InternalData|GetHNDL_Timer_ValidateTargetReachable))
  (Utilities|Time|ClearandInvalidateTimerbyHandle (Variables|InternalData|GetHNDL_Timer_GoingBackToNormal))
  (Utilities|IsValid _aicontroller
    (:Is%20Valid
      (AlertLevel|UnbindEventfromOnAlertLevelChangedEvent _aicontroller (EventDispatchers|CreateEvent :!event Custom|AIAlertLevelChanged))
      (Perception|UnbindEventfromOnCurrentTargetChangedEvent _aicontroller (EventDispatchers|CreateEvent :!event Custom|AICurrentTargetChanged)))))

(event Custom|AIAlertLevelChanged
  (params (BotController :type Fort%20Athena%20AIBot%20Controller%20Object%20Reference) (OldAlertLevel :type EAlertLevel%20Enum :default "Unaware") (NewAlertLevel :type EAlertLevel%20Enum :default "Unaware"))
  (if (Utilities|Enum|Equal(Enum) NewAlertLevel "Threatened")
    (CallFunction|HandleThreatened)
    (elif (Utilities|Enum|Equal(Enum) NewAlertLevel "Unaware")
      (CallFunction|HandleUnaware))))

(event Custom|HandleUnaware
  (bind _returntonormaltimer (Variables|Default|GetReturnToNormalTimer))
  (bind _returnvalue (ScalableFloat|GetValueatLevel _returntonormaltimer))
  (if (Variables|InternalData|GetIsMovingAway)
    (if (not (>= (- (Utilities|Time|GetGameTimeinSeconds) (Variables|InternalData|GetTimeWhenStartedMovingAway)) _returnvalue))
      (Utilities|FlowControl|Delay _returnvalue))
    (CallFunction|ValidateGoingBacktoNormal)))

(event Custom|HandleThreatened
  (bind _returnvalue (Utilities|Time|SetTimerbyEvent (EventDispatchers|CreateEvent :!event Custom|ValidateTargetReachable) (ScalableFloat|GetValueatLevel (Variables|Default|GetValidateReachableTimer)) true))
  (Variables|InternalData|SetHNDL_Timer_ValidateTargetReachable _returnvalue))

(event Custom|ValidateTargetReachable
  (bind _aicontroller (Variables|InternalData|GetAIController))
  (bind _targetactor (Class|BPFLAICommonFunctions|GetTargetActorfromBlackboard (Variables|InternalData|GetAIPawn)))
  (Utilities|IsValid _targetactor
    (:Is%20Valid
      (if (AI|Components|Blackboard|GetValueasBool (Variables|InternalData|GetAIBlackBoard) (Variables|Default|GetBBK_TargetReachable))
        (Variables|InternalData|SetStartedUnreachableCheck false)
        (else
          (bind _targetactor_2 (Class|BPFLAICommonFunctions|GetTargetActorfromBlackboard (Variables|InternalData|GetAIPawn)))
          (if (or (Movement|IsMovingonGround (Variables|/Script/Engine.Character|GetCharacterMovement _targetactor_2)) (WaterBody|IsSurfaceSwimming _targetactor_2))
            (if (Variables|InternalData|GetStartedUnreachableCheck)
              (if (>= (- (Utilities|Time|GetGameTimeinSeconds) (Variables|InternalData|GetTimeWhenStartedUnreachable)) (ScalableFloat|GetValueatLevel (Variables|Default|GetTimeToConsiderUnreachable)))
                (if (not (Variables|InternalData|GetIsMovingAway))
                  (bind _targetactor_3 (Class|BPFLAICommonFunctions|GetTargetActorfromBlackboard (Variables|InternalData|GetAIPawn)))
                  (Variables|InternalData|SetCurrentUnreachableTarget _targetactor_3)
                  (bind _output_get (Variables|InternalData|SetTempLeashCenter (CallFunction|GetTemporaryLeashCenterAwayfromTarget)))
                  (Class|BPFLAICommonFunctions|SetRadialLeash (Variables|InternalData|GetAIController) _output_get (ScalableFloat|GetValueatLevel (Variables|Default|GetInnerLeashRadiusWhenUnreachable)) (ScalableFloat|GetValueatLevel (Variables|Default|GetOutterLeashRadiusWhenUnreachable)) false)
                  (CallFunction|ResetLeashHack)
                  (Utilities|Time|ClearandInvalidateTimerbyHandle (Variables|InternalData|GetHNDL_Timer_ValidateTargetReachable))
                  (Class|FortAthenaAIBotController|PauseTargetFocusing _aicontroller)
                  (Class|FortAthenaAIBotController|SetAIFocalPoint _aicontroller 0 _output_get false)
                  (AI|PushBehaviorControl _aicontroller -1048573)
                  (Variables|InternalData|SetIsMovingAway true)
                  (Variables|InternalData|SetTimeWhenStartedMovingAway (Utilities|Time|GetGameTimeinSeconds))
                  (bind _returnvalue (Utilities|Time|SetTimerbyEvent (EventDispatchers|CreateEvent :!event Custom|ValidateGoingBackToNormal) (ScalableFloat|GetValueatLevel (Variables|Default|GetReturnToNormalTimer)) true))
                  (Variables|InternalData|SetHNDL_Timer_GoingBackToNormal _returnvalue)))
              (else
                (Variables|InternalData|SetStartedUnreachableCheck true)
                (Variables|InternalData|SetTimeWhenStartedUnreachable (Utilities|Time|GetGameTimeinSeconds))))))))))

(event Custom|GoBackToNormal
  (bind _aicontroller (Variables|InternalData|GetAIController))
  (Utilities|Time|ClearandInvalidateTimerbyHandle (Variables|InternalData|GetHNDL_Timer_GoingBackToNormal))
  (Class|BPFLAICommonFunctions|SetRadialLeash _aicontroller (Variables|InternalData|GetOGLeashCenter) (Variables|InternalData|GetOGLeashInnerRadius) (Variables|InternalData|GetOGLeashOutterRadius) false)
  (CallFunction|ResetLeashHack)
  (AI|PopBehaviorControl _aicontroller)
  (Class|FortAthenaAIBotController|ClearAIFocalPoint _aicontroller false)
  (Class|FortAthenaAIBotController|ResumeTargetFocusing _aicontroller)
  (Variables|InternalData|SetIsMovingAway false)
  (Variables|InternalData|SetStartedUnreachableCheck false)
  (Variables|InternalData|SetCurrentUnreachableTarget 0)
  (if (Variables|InternalData|GetReActivateThreatenedTimerOnReturnToNormal)
    (CallFunction|HandleThreatened)))

(event Custom|ValidateGoingBackToNormal
  (bind _aicontroller (Variables|InternalData|GetAIController))
  (Variables|InternalData|SetReActivateThreatenedTimerOnReturnToNormal false)
  ; Comment
  (bind _targetactor (Class|BPFLAICommonFunctions|GetTargetActorfromBlackboard (Variables|InternalData|GetAIPawn)))
  (Utilities|IsValid _targetactor
    (:Is%20Valid
      (if (Utilities|Enum|NotEqual(Enum) (Variables|/Script/FortniteAI.FortAthenaAIBotController|GetCurrentAlertLevel (Variables|InternalData|GetAIController)) "Threatened")
        _
        (elif (AI|Components|Blackboard|GetValueasBool (Variables|InternalData|GetAIBlackBoard) (Variables|Default|GetBBK_TargetReachable))
          (Variables|InternalData|SetReActivateThreatenedTimerOnReturnToNormal true)))
      (Utilities|Time|ClearandInvalidateTimerbyHandle (Variables|InternalData|GetHNDL_Timer_GoingBackToNormal))
      (Class|BPFLAICommonFunctions|SetRadialLeash _aicontroller (Variables|InternalData|GetOGLeashCenter) (Variables|InternalData|GetOGLeashInnerRadius) (Variables|InternalData|GetOGLeashOutterRadius) false)
      (CallFunction|ResetLeashHack)
      (AI|PopBehaviorControl _aicontroller)
      (Class|FortAthenaAIBotController|ClearAIFocalPoint _aicontroller false)
      (Class|FortAthenaAIBotController|ResumeTargetFocusing _aicontroller)
      (Variables|InternalData|SetIsMovingAway false)
      (Variables|InternalData|SetStartedUnreachableCheck false)
      (Variables|InternalData|SetCurrentUnreachableTarget 0)
      (if (Variables|InternalData|GetReActivateThreatenedTimerOnReturnToNormal)
        (CallFunction|HandleThreatened)))
    (:Is%20Not%20Valid
      (CallFunction|GoBacktoNormal))))

(event Custom|AICurrentTargetChanged
  (params (OldTarget :type Actor%20Object%20Reference) (NewTarget :type Actor%20Object%20Reference))
  (if (Variables|InternalData|GetIsMovingAway)
    (Utilities|Casting|CastToFortPawn NewTarget)
    (if (!= NewTarget (Variables|InternalData|GetCurrentUnreachableTarget))
      (CallFunction|GoBacktoNormal)
      (CallFunction|HandleThreatened))))

(fn CanReachTarget
  (return false))

(fn GetLeashData
  (bind _returnvalue (AI|GetorCreateAIRuntimeParameters (Variables|InternalData|GetAIController) "/Script/FortniteAI.FortAthenaAIRuntimeParameters_Leash"))
  (bind _asfort_athena_airuntime_parameters_leash (Utilities|Casting|CastToFortAthenaAIRuntimeParameters_Leash _returnvalue))
  (return (Class|FortAthenaAIRuntimeParametersLeash|GetLeashLocation _asfort_athena_airuntime_parameters_leash) (Class|FortAthenaAIRuntimeParametersLeash|GetLeashInnerRadius _asfort_athena_airuntime_parameters_leash) (Class|FortAthenaAIRuntimeParametersLeash|GetLeashOuterRadius _asfort_athena_airuntime_parameters_leash)))

(fn GetTemporaryLeashCenterAwayFromTarget
  (bind _targetactor (Class|BPFLAICommonFunctions|GetTargetActorfromBlackboard (Variables|InternalData|GetAIPawn)))
  (Utilities|IsValid _targetactor
    (:Is%20Valid
      (bind _output_get (LocalVariables|SetTargetPawn _targetactor))
      (return (+ (Transformation|GetActorLocation (Variables|InternalData|GetAIPawn)) (* (Math|Vector|Normalize (- (Transformation|GetActorLocation (Variables|InternalData|GetAIPawn)) (Transformation|GetActorLocation _output_get))) (ScalableFloat|GetValueatLevel (Variables|Default|GetDIstanceToMoveAwayPawnWhenUnreachable))))))
    (:Is%20Not%20Valid
      (return "0, 0, 0"))))

(fn ResetLeashHack
  (AI|Components|Blackboard|SetValueasEnum (AI|GetBlackboard (Variables|InternalData|GetAIController)) (Variables|Default|GetBBKeyExecutionStatus) 0)
  (return))
